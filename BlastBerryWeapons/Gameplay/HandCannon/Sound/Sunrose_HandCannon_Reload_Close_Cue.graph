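# Sound cue for the Sunrose HandCannon weapon's close reload sound, built from reload "pull" samples.
# A Random node picks one of three equally weighted wave variations and feeds it through a default Modulator for slight variation.
# It plays at reduced volume with weapon reload attenuation, under a third-person ranged-fire sound class.
# A concurrency override blocks new instances once the limit is reached rather than cutting off playing ones.
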SoundCue /BlastBerryWeapons/Gameplay/HandCannon/Sound/Sunrose_HandCannon_Reload_Close_Cue
root=Modulator_1
nodes [
 {id=Modulator_1, children [Random_0]}
 {id=Random_0, children [WavePlayer_0,WavePlayer_1,WavePlayer_2], properties {weights [1,1,1]}}
 {id=WavePlayer_0, properties {soundWaveAssetPtr=&/SunRose_HandCannon_Reload_Pull_01}}
 {id=WavePlayer_1, properties {soundWaveAssetPtr=&/SunRose_HandCannon_Reload_Pull_02}}
 {id=WavePlayer_2, properties {soundWaveAssetPtr=&/SunRose_HandCannon_Reload_Pull_03}}
]
properties {
 attenuationSettings=&Fort_Weapon_Reload_Att
 bOverrideConcurrency=true
 maxDistance=2250
 soundClassObject=&Fort_WeaponRanged_Fire_3P_PSM
 volumeMultiplier=0.3
 concurrencyOverrides {resolutionRule=PreventNew}
}
