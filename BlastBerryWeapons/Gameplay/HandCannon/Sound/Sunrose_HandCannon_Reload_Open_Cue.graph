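# Sound cue for the reload-open foley of the Sunrose hand cannon weapon.
# A single wave player reuses a stock pistol eject foley sound and feeds a Modulator left at its default pitch and volume variation.
# The cue plays at reduced volume with weapon reload attenuation and a third-person ranged-fire sound class.
# Its concurrency override prevents new instances once the limit is reached.
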SoundCue /BlastBerryWeapons/Gameplay/HandCannon/Sound/Sunrose_HandCannon_Reload_Open_Cue
root=Modulator_0
nodes [
 {id=Modulator_0, children [WavePlayer_11]}
 {id=WavePlayer_11, properties {soundWaveAssetPtr=/Game/Sounds/Fort_Weapons/pistol/pistol_foley/auto_sdk_eject_01}}
]
properties {
 attenuationSettings=&Fort_Weapon_Reload_Att
 bOverrideConcurrency=true
 maxDistance=2250
 soundClassObject=&Fort_WeaponRanged_Fire_3P_PSM
 volumeMultiplier=0.4
 concurrencyOverrides {resolutionRule=PreventNew}
}
